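# Material function for the Helix vehicle boost that outputs a radial Opacity mask, named for a chroma blur effect.
# The mask peaks at the UV center, built from inverted squared distance raised to a BlurFalloff power and clamped.
# An optional AddNoise static switch multiplies it by noise panned over time in radial coordinates.
# A nested ExtraNoise switch adds a second noise sample with its own tiling and speed multipliers.
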MaterialFunction /VehicleCosmetics/Boosts/Boost_Helix/Materials/MF_Boost_ChromaBlur
graph {nodes [
  n1 = Add(A=TextureCoordinate(), B=Constant() {r=-0.5})
  n2 = DotProduct(A=@n1, B=@n1)
  n3 = Constant() {r=8}
  MaterialExpressionConstantClamp_3 = Saturate(None=Power(Base=OneMinus(None=@n2), Exp=FunctionInput(Preview=@n3) {inputName=BlurFalloff, inputType=FunctionInput_Scalar}))
  n4 = Constant() {r=3}
  n5 = Constant()
  n6 = Multiply(
   A=MaterialFunctionCall() {materialFunction=&VectorToRadialValue}.Radial Coordinates
   B=AppendVector(
    A=FunctionInput(Preview=@n4) {description=Default Value is 3, inputName=NoiseTile_X, inputType=FunctionInput_Scalar}
    B=FunctionInput(Preview=@n5) {description=Default Value is 0, inputName=NoiseTile_Y, inputType=FunctionInput_Scalar}
   )
  )
  n7 = AppendVector(
   A=FunctionInput(Preview=Constant() {r=0.05}) {bUsePreviewValueAsDefault=true, description=Default Value is .05, inputName=PanSpeed_X, inputType=FunctionInput_Scalar}
   B=FunctionInput(Preview=Constant() {r=-0.1}) {bUsePreviewValueAsDefault=true, description=Default Value is -0.1, inputName=PanSpeed_Y, inputType=FunctionInput_Scalar}
  )
  n8 = Time()
  n9 = AppendVector(A=@n8, B=@n8)
  n10 = TextureObject() {texture=/DealershipSharedLibrary/FX/Textures/Noise/T_VCVFX_Noise_Grass01_Alpha}
  n11 = TextureSample(
   UVs=Add(A=@n6, B=Multiply(A=@n7, B=@n9))
   Tex=FunctionInput(Preview=@n10) {bUsePreviewValueAsDefault=true, inputName=NoiseTexture, inputType=FunctionInput_Texture2D}
  )
  n12 = FunctionInput(Preview=Constant() {r=1.5}) {bUsePreviewValueAsDefault=true, description=Default Value is 1.5, inputName=ExtraNoiseTile, inputType=FunctionInput_Scalar}
  n13 = FunctionInput(Preview=Constant() {r=0.6}) {bUsePreviewValueAsDefault=true, description=Default Value is .6, inputName=ExtraNoiseSpeedMult, inputType=FunctionInput_Scalar}
  n14 = AppendVector(A=@n13, B=@n13)
  n15 = Add(A=Multiply(A=@n6, B=@n12), B=Multiply(A=Multiply(A=@n7, B=@n14), B=@n9))
  n16 = FunctionInput(Preview=@n10) {bUsePreviewValueAsDefault=true, inputName=Extra_NoiseTexture, inputType=FunctionInput_Texture2D}
  n17 = StaticSwitch(
   True=Add(A=@n11.G, B=TextureSample(UVs=@n15, Tex=@n16).G)
   False=@n11.G
   Value=FunctionInput(Preview=StaticBool()) {bUsePreviewValueAsDefault=true, inputName=ExtraNoise, inputType=FunctionInput_StaticBool}
  )
  n18 = StaticBool()
  FunctionOutput(None=StaticSwitch(
    True=Multiply(A=@MaterialExpressionConstantClamp_3, B=@n17)
    False=@MaterialExpressionConstantClamp_3
    Value=FunctionInput(Preview=@n18) {bUsePreviewValueAsDefault=true, inputName=AddNoise, inputType=FunctionInput_StaticBool}
   )) {outputName=Opacity}
 ]}
