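# Unlit translucent Niagara sprite and ribbon material for the core chroma blur particle of the Helix vehicle boost.
# MF_Boost_ChromaBlur pans noise on Y, with dynamic parameters driving speed, blur falloff and a contrast exponent (OpacityThreshold).
# The result, tinted by particle color with a small lift, passes through shared emissive adjustments.
# A separate radial UV mask, depth-faded and multiplied by particle alpha, forms the opacity term.
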Material /VehicleCosmetics/Boosts/Boost_Helix/Materials/M_Boost_Core_ChromaBlurParticle
graph {
 properties {
  bAutomaticallySetUsageInEditor=false
  bUsedWithNiagaraRibbons=true
  bUsedWithNiagaraSprites=true
  blendMode=BLEND_Translucent
  shadingModel=MSM_Unlit
  usedShadingModels=Unlit
 }
 nodes [
  MaterialExpressionVertexColor_0 = ParticleColor()
  n1 = DynamicParameter() {paramNames [Intensity, Scale, Speed_Y, OpacityThreshold]}
  n2 = Power(
   Base=MaterialFunctionCall!(
    ExtraNoiseTile=ScalarParameter() {defaultValue=1, parameterName=ExtraNoiseTile}
    NoiseTile_Y=ScalarParameter() {parameterName=NoiseTile_Y}
    PanSpeed_Y=ComponentMask(None=@n1.Speed_Y) {r=true}
    PanSpeed_X=Constant()
    ExtraNoiseSpeedMult=ScalarParameter() {defaultValue=1, parameterName=ExtraNoiseSpeedMult}
    NoiseTile_X=ScalarParameter() {defaultValue=1, parameterName=NoiseTile_X}
    ExtraNoise=StaticBoolParameter() {parameterName=ExtraNoise}
    BlurFalloff=ComponentMask(None=@n1.Scale) {r=true}
    AddNoise=StaticBoolParameter() {defaultValue=true, parameterName=Noise}
    NoiseTexture=TextureObjectParameter() {parameterName=NoiseTexture, texture=/DealershipSharedLibrary/FX/Textures/Noise/T_VCVFX_Caustics_A}
    Extra_NoiseTexture=TextureObjectParameter() {parameterName=ExtraNoiseTexture, texture=&T_VCVFX_Caustics_A}
   ) {materialFunction=&/MF_Boost_ChromaBlur}.Opacity
   Exp=@n1.OpacityThreshold
  )
  MaterialFunctionCall(Emissive=Add(A=Multiply(A=@MaterialExpressionVertexColor_0.RGB, B=@n2)) {constB=0.1}) {materialFunction=&MF_FX_EmissiveAdjustments}
  n3 = Add(A=TextureCoordinate(), B=Constant() {r=-0.5})
  n4 = ComponentMask(None=@n3) {g=true}
  n5 = ComponentMask(None=@n3) {r=true}
  n6 = Add(A=Multiply(A=@n4, B=@n4), B=Multiply(A=@n5, B=@n5))
  n7 = ScalarParameter() {parameterName=CenterMaskRadius}
  MaterialExpressionConstantClamp_4 = Saturate(None=Multiply(A=Subtract(A=@n6, B=@n7), B=ScalarParameter() {defaultValue=6, parameterName=CenterMaskSharpness}))
  Multiply(
   A=@MaterialExpressionVertexColor_0.A
   B=MaterialFunctionCall(Opacity In=OneMinus(None=@MaterialExpressionConstantClamp_4), DepthFade Distance=Constant() {r=16}) {materialFunction=&MF_VCVFX_DepthFade}.Result
  )
 ]
}
