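# Translucent, unlit, two-sided Niagara material for an illustrative-style helix vehicle boost effect.
# A panning texture supplies power-shaped, particle-tinted emissive through MF_FX_EmissiveAdjustments, and its alpha, faded along U, drives opacity.
# Optional world position offset builds a spinning helix from time-shifted sine and cosine in Y/Z, with switchable direction and UV1 masking.
# Dynamic parameters let particles control emissive boost, color add, spin phase and mask strength.
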Material /VehicleCosmetics/Boosts/Boost_Helix/Materials/M_Boost_Illustrative
graph {
 properties {
  bAutomaticallySetUsageInEditor=false
  bUsedWithNiagaraMeshParticles=true
  bUsedWithNiagaraRibbons=true
  bUsedWithNiagaraSprites=true
  blendMode=BLEND_Translucent
  shadingModel=MSM_Unlit
  twoSided=true
  usedShadingModels=Unlit
 }
 nodes [
  MaterialExpressionVertexColor_1 = ParticleColor()
  n1 = TextureCoordinate()
  n2 = AppendVector(A=ScalarParameter() {defaultValue=1.5, parameterName=Tile U}, B=ScalarParameter() {defaultValue=1, parameterName=Tile V})
  n3 = Multiply(A=Time(), B=ScalarParameter() {defaultValue=1, parameterName=TimeMult})
  n4 = ScalarParameter() {defaultValue=2, parameterName=TimePeriod}
  n5 = TextureSampleParameter2D(UVs=Panner(Coordinate=Multiply(A=@n1, B=@n2), Time=Fmod(A=@n3, B=@n4)) {speedX=-2}) {parameterName=MainTex Param, texture=/DealershipSharedLibrary/FX/Textures/Gradient_Masks/T_VCVFX_Illustrative_Alpha}
  n6 = TextureCoordinate()
  n7 = Multiply(
   A=Power(Base=ComponentMask(None=@n6) {r=true}, Exp=ScalarParameter() {defaultValue=1.3, parameterName=opa})
   B=ScalarParameter() {defaultValue=10, parameterName=opachinh}
  )
  MaterialExpressionConstantClamp_10 = Saturate(None=Multiply(A=@MaterialExpressionVertexColor_1.A, B=Multiply(A=@n5.A, B=@n7)))
  n8 = DynamicParameter() {defaultValue=<rgba 1 0.6 1 4>, paramNames [EmissiveBoost, SubUVTimeoffset, RibbonColorAdd, WPO_Mask_Mult]}
  n9 = ScalarParameter() {defaultValue=1.5, parameterName=Multiplier}
  MaterialExpressionConstantClamp_12 = Saturate(None=@n5.R)
  n10 = Power(Base=@MaterialExpressionConstantClamp_12, Exp=ScalarParameter() {defaultValue=0.09, parameterName=PowerExp1})
  n11 = Multiply(A=Power(Base=Multiply(A=@n9, B=@n10), Exp=ScalarParameter() {defaultValue=3.5, parameterName=PowerExp2}), B=@n10)
  n12 = Desaturation(None=@n11) {luminanceFactors=<rgba 0.33 0.59 0.11 0>}
  n13 = Multiply(
   A=VectorParameter() {parameterName=BloomMult, defaultValue=<rgba 1 1 1 1>}.RGB
   B=Multiply(A=@MaterialExpressionVertexColor_1.RGB, B=StaticSwitchParameter(True=@n12, False=@n11) {parameterName=DestaturateForThumbnail})
  )
  MaterialExpressionConstantClamp_9 = Saturate(None=Add(A=@n8.RibbonColorAdd, B=Multiply(A=@n8.EmissiveBoost, B=@n13)))
  MaterialFunctionCall(Emissive=@MaterialExpressionConstantClamp_9) {materialFunction=&MF_FX_EmissiveAdjustments}
  ScalarParameter() {defaultValue=4, parameterName=WPO_Mask_Mult}
  ScalarParameter() {parameterName=SinUVTimeOffset}
  n14 = TextureCoordinate()
  n15 = ScalarParameter() {defaultValue=1, parameterName=SinUVTile}
  n16 = Multiply(A=ScalarParameter() {defaultValue=1, parameterName=SinUVTimeMult}, B=Time())
  n17 = DynamicParameter() {defaultValue=<rgba 1 0.6 1 4>, paramNames [EmissiveBoost, SubUVTimeoffset, RibbonColorAdd, WPO_Mask_Mult]}
  n18 = Add(A=ComponentMask(None=Multiply(A=@n14, B=@n15)) {r=true}, B=Fmod(A=Add(A=@n16, B=@n17.SubUVTimeoffset), B=Constant() {r=1}))
  n19 = Sine(None=@n18)
  n20 = Cosine(None=@n18)
  n21 = AppendVector(
   A=AppendVector(A=Constant(), B=StaticSwitchParameter(True=@n19, False=@n20) {parameterName=Spin_CCW})
   B=StaticSwitchParameter(True=@n20, False=@n19) {parameterName=Spin_CCW}
  )
  n22 = ScalarParameter() {defaultValue=1, parameterName=WPO Mult}
  n23 = Power(
   Base=ComponentMask(None=TextureCoordinate() {coordinateIndex=1}) {r=true}
   Exp=ScalarParameter() {defaultValue=0.5, parameterName=WPO_Mask_Power}
  )
  MaterialExpressionConstantClamp_11 = Saturate(None=Add(A=ScalarParameter() {defaultValue=0.25, parameterName=WPO_Mask_Min}, B=Multiply(A=@n23, B=@n17.WPO_Mask_Mult)))
  n24 = Constant() {r=1}
  StaticSwitchParameter(
   True=Multiply(A=Multiply(A=@n21, B=@n22), B=StaticSwitchParameter(True=@MaterialExpressionConstantClamp_11, False=@n24) {parameterName=MaskWPO})
   False=Constant()
  ) {parameterName=UseWPO}
 ]
}
