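# Translucent unlit trail material for the Helix vehicle boost, usable on Niagara sprites, ribbons and meshes.
# Panning trail, wind-particle and doubled star-beam noise textures are shaped by U falloffs and top/bottom gradient fades into opacity and color masks.
# Dynamic parameters set falloff exponents and particle-color power, and blend fixed green toward particle color via "desaturation".
# Summed color feeds an emissive-adjustment function.
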Material /VehicleCosmetics/Boosts/Boost_Helix/Materials/M_Boost_Trail_option3
graph {
 properties {
  bAutomaticallySetUsageInEditor=false
  bUsedWithNiagaraMeshParticles=true
  bUsedWithNiagaraRibbons=true
  bUsedWithNiagaraSprites=true
  blendMode=BLEND_Translucent
  shadingModel=MSM_Unlit
  usedShadingModels=Unlit
 }
 nodes [
  n1 = MaterialFunctionCall!(
   "Gradient U=0 V=1"=Constant() {r=1}
   "Invert gradient 0=False 1=True"
   "Contrast (Power)"=ScalarParameter() {defaultValue=2.8, group=Opacity, parameterName=Contrast Fade Top, sortPriority=9}
   "Brightness (Multiply)"=ScalarParameter() {defaultValue=100, group=Opacity, parameterName=Brightness Fade Top, sortPriority=10}
   "Dynamic Contrast (Power)"
   "Dynamic Brightness (Multiply)"
  ) {materialFunction=/DealershipSharedLibrary/FX/Materials/MaterialFunctions/MF_LinearGradientPlus}
  Ceil(None=@n1.Result)
  n2 = Time()
  n3 = ScalarParameter() {defaultValue=1}
  n4 = TextureSample(UVs=Panner(Coordinate=TextureCoordinate(), Time=Multiply(A=@n2, B=@n3)) {speedX=-1}) {texture=/DealershipSharedLibrary/FX/Textures/Gradient_Masks/T_VCVFX_Trail3_Alpha}
  n5 = TextureCoordinate()
  n6 = Power(
   Base=OneMinus(None=ComponentMask(None=@n5) {r=true})
   Exp=DynamicParameter() {defaultValue=<rgba 1 1 8 10>, paramNames [desaturation,power,Mascara,manchas]}.manchas
  )
  n7 = Constant() {r=1}
  n8 = Power(Base=ComponentMask(None=TextureCoordinate()) {r=true}, Exp=Constant() {r=2})
  n9 = Constant() {r=500}
  MaterialExpressionConstantClamp_8 = Saturate(None=Multiply(A=Multiply(A=@n6, B=@n7), B=Multiply(A=@n8, B=@n9)))
  MaterialExpressionConstantClamp_7 = Saturate(None=Multiply(A=@n4.R, B=@MaterialExpressionConstantClamp_8))
  MaterialExpressionVertexColor_1 = ParticleColor()
  n10 = Constant() {r=20}
  n11 = Power(Base=@n1.Result, Exp=@n10) {constExponent=20}
  n12 = Constant() {r=1}
  n13 = Power(
   Base=MaterialFunctionCall(
    "Gradient U=0 V=1"=@n12
    "Invert gradient 0=False 1=True"=@n12
    "Contrast (Power)"=ScalarParameter() {defaultValue=2.8, group=Opacity, parameterName=Contrast Fade Bottom, sortPriority=11}
    "Brightness (Multiply)"=ScalarParameter() {defaultValue=100, group=Opacity, parameterName=Brightness Fade Bottom, sortPriority=12}
   ) {materialFunction=&MF_LinearGradientPlus}.Result
   Exp=@n10
  ) {constExponent=20}
  n14 = Multiply(A=Saturate(None=Multiply(A=@n11, B=@n13)), B=ScalarParameter() {defaultValue=0.8, parameterName=Intensity_Mask}) {constB=0.8}
  n15 = Multiply(A=@MaterialExpressionVertexColor_1.A, B=@n14) {constA=1}
  n16 = ComponentMask(None=TextureCoordinate()) {r=true}
  n17 = DynamicParameter() {defaultValue=<rgba 1 1 8 10>, paramNames [desaturation,power,Mascara,manchas]}
  n18 = Multiply(A=Power(Base=OneMinus(None=@n16), Exp=@n17.Mascara), B=Constant() {r=1})
  n19 = Multiply(A=ScalarParameter() {defaultValue=1, parameterName=MaskSpeed}, B=Time())
  n20 = TextureSample(UVs=Panner(Coordinate=TextureCoordinate() {uTiling=2, vTiling=0.25}, Time=@n19) {speedX=-1, speedY=-0.1}) {texture=/VehicleCosmetics/FX/Textures/Noise/T_VCVFX_NoiseStarBeam_A}
  n21 = Constant() {r=1.5}
  n22 = TextureSample(UVs=Panner(Coordinate=TextureCoordinate() {vTiling=0.5}, Time=@n19) {speedX=-1, speedY=-0.1}) {texture=&T_VCVFX_NoiseStarBeam_A}
  n23 = Constant() {r=1.5}
  MaterialExpressionConstantClamp_11 = Saturate(None=Add(A=Multiply(A=@n20.RGB, B=@n21), B=Multiply(A=@n22.RGB, B=@n23)))
  MaterialExpressionConstantClamp_10 = Saturate(None=Add(A=@n18, B=Multiply(A=@n18, B=@MaterialExpressionConstantClamp_11)))
  n24 = TextureCoordinate()
  n25 = Multiply(A=Power(Base=ComponentMask(None=@n24) {r=true}, Exp=Constant() {r=1.5}), B=Constant() {r=500})
  MaterialExpressionConstantClamp_9 = Saturate(None=Multiply(A=@MaterialExpressionConstantClamp_10, B=@n25))
  MaterialExpressionConstantClamp_6 = Saturate(None=Add(A=@MaterialExpressionConstantClamp_7, B=Multiply(A=@n15, B=@MaterialExpressionConstantClamp_9)))
  Constant() {r=10}
  Constant() {r=10}
  n26 = Multiply(A=Multiply(A=@n4.R, B=Constant3Vector() {constant=<rgba 0.212 1 0.0987 0>}), B=ScalarParameter() {defaultValue=1.3, parameterName=Color})
  n27 = DynamicParameter() {defaultValue=<rgba 1 1 8 10>, paramNames [desaturation,power,Mascara,manchas]}
  n28 = Multiply(A=Multiply(A=@MaterialExpressionVertexColor_1.RGB, B=@n4.R), B=@n27.power)
  n29 = Time()
  n30 = ScalarParameter() {defaultValue=1}
  n31 = TextureSample(UVs=Panner(Coordinate=TextureCoordinate() {uTiling=2, vTiling=0.3}, Time=Multiply(A=@n29, B=@n30)) {speedX=-1}) {texture=/DealershipSharedLibrary/FX/Textures/Gradient_Masks/T_VCVFX_Boost_Helix_WindParticles_Alpha}
  n32 = ComponentMask(None=TextureCoordinate()) {r=true}
  n33 = Multiply(A=Power(Base=OneMinus(None=@n32), Exp=@n17.manchas), B=Constant() {r=1})
  MaterialExpressionConstantClamp_13 = Saturate(None=Add(A=@n33, B=Multiply(A=@n33, B=@MaterialExpressionConstantClamp_11)))
  MaterialExpressionConstantClamp_12 = Saturate(None=Multiply(A=@MaterialExpressionConstantClamp_13, B=@n25))
  n34 = Multiply(A=@n31.R, B=Multiply(A=@n15, B=@MaterialExpressionConstantClamp_12))
  n35 = Constant3Vector() {constant=<rgba 0.212 1 0.0987 0>}
  n36 = Multiply(A=@n14, B=@n31.R)
  n37 = LinearInterpolate(
   A=Multiply(A=Multiply(A=@n34, B=@n35), B=Constant() {r=2})
   B=Multiply(A=Multiply(A=@MaterialExpressionVertexColor_1.RGB, B=@n36), B=@n27.power)
   Alpha=@n27.desaturation
  )
  n38 = Multiply(
   A=@MaterialExpressionVertexColor_1.RGB
   B=Desaturation(None=@n14, Fraction=ScalarParameter() {parameterName=Painted}) {luminanceFactors=<rgba 0.33 0.59 0.11 0>}
  )
  MaterialExpressionConstantClamp_5 = Saturate(None=Add(A=LinearInterpolate(A=@n26, B=@n28, Alpha=@n27.desaturation), B=Add(A=@n37, B=@n38)))
  MaterialFunctionCall(Emissive=@MaterialExpressionConstantClamp_5) {materialFunction=&MF_FX_EmissiveAdjustments}
  ScalarParameter() {defaultValue=8, parameterName=Mascara}
  ScalarParameter() {defaultValue=2, parameterName=Manchas}
  Step()
  TextureSample()
 ]
}
